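# Sound cue for the decoy grenade's clone death sound, in the consumables sound class.
# A short fixed delay feeds a default-setting modulator, then an equally weighted random pick of three wave variations, exempt from branch culling.
# It references the decoy hologram attenuation asset for spatial falloff.
# A concurrency override caps it at two simultaneous instances, taking precedence over the listed hit-notification concurrency set.
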SoundCue /Game/Sounds/Fort_Weapons/Grenades/Decoy/DecoyGrenade_CloneDeath_Cue
root=Delay_0
nodes [
 {id=Delay_0, children [Modulator_0], properties {delayMax=0.1, delayMin=0.1}}
 {id=Modulator_0, children [Random_1]}
 {id=Random_1, children [WavePlayer_2,WavePlayer_3,WavePlayer_4], properties {bShouldExcludeFromBranchCulling=true, weights [1,1,1]}}
 {id=WavePlayer_2, properties {soundWaveAssetPtr=&/DecoyGrenade_CloneDeath_01}}
 {id=WavePlayer_3, properties {soundWaveAssetPtr=&/DecoyGrenade_CloneDeath_02}}
 {id=WavePlayer_4, properties {soundWaveAssetPtr=&/DecoyGrenade_CloneDeath_03}}
]
properties {
 attenuationSettings=&/DecoyGrenade_Hologram_Att
 bOverrideConcurrency=true
 maxDistance=6300
 soundClassObject=&Fort_Consumables
 volumeMultiplier=0.9
 attenuationOverrides {falloffDistance=3000, attenuationShapeExtents {x=500}}
 concurrencyOverrides {maxCount=2}
 concurrencySet [&HitNotifications_TwoMaximum_StopOldest_VolumeScale]
}
