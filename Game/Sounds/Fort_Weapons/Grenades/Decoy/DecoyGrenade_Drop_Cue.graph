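# Sound cue for the Decoy Grenade drop sound, assembled from drop samples borrowed from other weapons.
# A mixer layers a boosted Air Strike (AppleSauce) drop with a quieter Dual Pistols drop, which a modulator pitches slightly down by a fixed amount.
# The cue plays at reduced volume through shared weapon-drop attenuation, the loot-collect sound class and a weapon-drop stop-oldest concurrency set.
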SoundCue /Game/Sounds/Fort_Weapons/Grenades/Decoy/DecoyGrenade_Drop_Cue
root=Mixer_0
nodes [
 {id=Mixer_0, children [WavePlayer_0,Modulator_0], properties {inputVolume [1.6,0.4]}}
 {id=Modulator_0, children [WavePlayer_2], properties {pitchMax=0.9, pitchMin=0.9}}
 {id=WavePlayer_0, properties {soundWaveAssetPtr=/Game/Athena/Sounds/Weapons/Air_Strike/AppleSauce_Drop_01}}
 {id=WavePlayer_2, properties {soundWaveAssetPtr=/Game/Sounds/Fort_Weapons/pistol/Dual/DualPistols_Drop_02}}
]
properties {
 attenuationSettings=&Fort_Weapon_Drop_Att
 maxDistance=3000
 soundClassObject=&Fort_Loot_Collect
 volumeMultiplier=0.4
 concurrencyOverrides {resolutionRule=PreventNew}
 concurrencySet [&Weapon_Drop_StopOldest]
}
