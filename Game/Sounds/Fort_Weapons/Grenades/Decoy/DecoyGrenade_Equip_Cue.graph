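# Sound cue for equipping the decoy grenade, played as player foley.
# A random node picks one of three equip wave variations with equal weight.
# It uses close-range prop attenuation, slightly lowered pitch and raised volume.
# A concurrency override prevents new instances from starting once the limit is reached.
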SoundCue /Game/Sounds/Fort_Weapons/Grenades/Decoy/DecoyGrenade_Equip_Cue
root=Random_0
nodes [
 {id=Random_0, children [WavePlayer_1,WavePlayer_2,WavePlayer_3], properties {weights [1,1,1]}}
 {id=WavePlayer_1, properties {soundWaveAssetPtr=&/DecoyGrenade_Equip_01}}
 {id=WavePlayer_2, properties {soundWaveAssetPtr=&/DecoyGrenade_Equip_02}}
 {id=WavePlayer_3, properties {soundWaveAssetPtr=&/DecoyGrenade_Equip_03}}
]
properties {
 attenuationSettings=/Game/Sounds/Attenuation/Fort_Prop_Close_Att
 bOverrideConcurrency=true
 maxDistance=1800
 pitchMultiplier=0.9
 soundClassObject=&Fort_Player_Foley
 volumeMultiplier=1.2
 concurrencyOverrides {resolutionRule=PreventNew}
}
