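# Sound cue for the decoy grenade detonation, layering an explosion with a hologram clone-spawn sound.
# One branch picks randomly from three explosion waves under explosion attenuation, mixed at half volume.
# The other picks one of two clone-spawn waves, raises its pitch and delays it briefly under separate hologram attenuation.
# The cue uses the medium explosives sound class and overrides concurrency to cap simultaneous instances.
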SoundCue /Game/Sounds/Fort_Weapons/Grenades/Decoy/DecoyGrenade_Explode_Cue
root=Mixer_2
nodes [
 {id=Attenuation_1, children [Random_0], properties {attenuationSettings=&/DecoyGrenade_Explo_Att}}
 {id=Attenuation_2, children [Delay_1], properties {attenuationSettings=&/DecoyGrenade_Hologram_Att}}
 {id=Delay_1, children [Modulator_0], properties {delayMax=0.1, delayMin=0.1}}
 {id=Mixer_2, children [Modulator_12,null], properties {inputVolume [1,0.8]}}
 {id=Mixer_3, children [Attenuation_1, Attenuation_2], properties {inputVolume [0.5,1]}}
 {id=Modulator_0, children [Random_6], properties {pitchMax=1.1, pitchMin=1.1}}
 {id=Modulator_12, children [Mixer_3], properties {pitchMax=1, pitchMin=1}}
 {id=Random_0, children [WavePlayer_0,WavePlayer_1,WavePlayer_2], properties {bShouldExcludeFromBranchCulling=true, weights [1,1,1]}}
 {id=Random_6, children [WavePlayer_16, WavePlayer_18], properties {bShouldExcludeFromBranchCulling=true, weights [1,1]}}
 {id=WavePlayer_0, properties {soundWaveAssetPtr=&/DecoyGrenade_Explosion_01}}
 {id=WavePlayer_1, properties {soundWaveAssetPtr=&/DecoyGrenade_Explosion_02}}
 {id=WavePlayer_16, properties {soundWaveAssetPtr=&/DecoyGrenade_CloneSpawn_01}}
 {id=WavePlayer_18, properties {soundWaveAssetPtr=&/DecoyGrenade_CloneSpawn_02}}
 {id=WavePlayer_2, properties {soundWaveAssetPtr=&/DecoyGrenade_Explosion_03}}
]
properties {
 bOverrideConcurrency=true
 maxDistance=13000
 soundClassObject=/Game/Sounds/FortSoundClassesAndModes/Fort_Explosives_Medium
 volumeMultiplier=0.7
 attenuationOverrides {
  bAttenuateWithLPF=true
  falloffDistance=3400
  lPFFrequencyAtMax=400
  lPFFrequencyAtMin=10000
  lPFRadiusMax=1500
  lPFRadiusMin=1000
  stereoSpread=2000
  attenuationShapeExtents {x=700}
 }
 concurrencyOverrides {maxCount=3}
}
