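# Sound cue for the Decoy Grenade's fuse, layering a fuse-start sound with a looping fuse sound in the gadgets sound class.
# A mixer plays both together, with the start layer slightly quieter and the loop pitched up by a fixed amount.
# Each layer has its own attenuation node, the loop's borrowed from a Bucket item's in-air projectile loop.
# Overridden concurrency prevents new instances once the limit is reached.
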SoundCue /Game/Sounds/Fort_Weapons/Grenades/Decoy/DecoyGrenade_FuseLoop_Cue
root=Mixer_1
nodes [
 {id=Attenuation_0, children [Modulator_0], properties {attenuationSettings=/Game/Athena/Sounds/Items/Bucket/Bucket_Nice_Prj_InAir_Loop_Att}}
 {id=Attenuation_1, children [WavePlayer_1], properties {attenuationSettings=/Game/Sounds/Attenuation/Fort_Explosive_Fuse_Prj_Throw_Att}}
 {id=Mixer_1, children [Attenuation_1, Attenuation_0], properties {inputVolume [0.8,1]}}
 {id=Modulator_0, children [WavePlayer_2], properties {pitchMax=1.1, pitchMin=1.1}}
 {id=WavePlayer_1, properties {soundWaveAssetPtr=&/DecoyGrenade_FuseStart_01}}
 {id=WavePlayer_2, properties {bLooping=true, soundWaveAssetPtr=&/DecoyGrenade_FuseLoop_01}}
]
properties {
 bOverrideConcurrency=true
 maxDistance=17000
 soundClassObject=&Fort_Gadgets
 volumeMultiplier=0.7
 attenuationOverrides {distanceAlgorithm=Logarithmic, falloffDistance=1200, attenuationShapeExtents {x=500}}
 concurrencyOverrides {resolutionRule=PreventNew}
}
