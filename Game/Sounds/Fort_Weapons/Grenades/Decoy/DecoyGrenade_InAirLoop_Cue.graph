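# Sound cue for the looping in-air sound of the Decoy Grenade.
# It mixes two looping waves, a Creative building grenade loop and a slightly quieter BattleLab bot grenade loop, through an Enveloper that fades the volume in briefly.
# The cue is pitched up and lowered in volume, borrows the frag grenade in-air attenuation, overrides concurrency to cap simultaneous instances, and plays in the medium explosives sound class.
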SoundCue /Game/Sounds/Fort_Weapons/Grenades/Decoy/DecoyGrenade_InAirLoop_Cue
root=Enveloper_0
nodes [
 {
  id=Enveloper_0
  children [Mixer_0]
  properties {pitchCurve {editorCurveData {keys [{time=0}]}}, volumeCurve {editorCurveData {keys [<at 0 0>,<at 0.3 1>]}}}
 }
 {id=Mixer_0, children [WavePlayer_1,WavePlayer_0], properties {inputVolume [0.7,1]}}
 {id=WavePlayer_0, properties {bLooping=true, soundWaveAssetPtr=/Game/Sounds/Creative/Gadgets/Building/Building_Grenade_InAirLoop01}}
 {id=WavePlayer_1, properties {bLooping=true, soundWaveAssetPtr=/Game/Sounds/BattleLab/BotGrenade/BotGrenade_InAirLoop_UnFriendly_01}}
]
properties {
 attenuationSettings=/Game/Sounds/Fort_Weapons/Grenades/Frag/Grenade_Frag_InAir_Att
 bOverrideConcurrency=true
 maxDistance=6300
 pitchMultiplier=1.2
 soundClassObject=/Game/Sounds/FortSoundClassesAndModes/Fort_Explosives_Medium
 volumeMultiplier=0.6
 attenuationOverrides {falloffDistance=5000, reverbWetLevelMax=0.7, reverbWetLevelMin=0.2, attenuationShapeExtents {x=1000}}
 concurrencyOverrides {maxCount=3}
}
