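# DecoyGrenade_Pickup_Cue is the pickup sound cue for the decoy grenade.
# A single wave player feeds the root directly, reusing the TimedBomb gadget's second pickup wave at reduced volume, with no randomization or modulation.
# It uses the shared weapon pickup attenuation and loot collect sound class, with a concurrency override that allows one instance and blocks new ones while it plays.
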SoundCue /Game/Sounds/Fort_Weapons/Grenades/Decoy/DecoyGrenade_Pickup_Cue
root=WavePlayer_4
nodes [{id=WavePlayer_4, properties {soundWaveAssetPtr=/Game/Sounds/Fort_Gadgets/TimedBomb/TimedBomb_Pickup_02}}]
properties {
 attenuationSettings=&Fort_Weapon_Pickup_Att
 bOverrideConcurrency=true
 maxDistance=3500
 soundClassObject=&Fort_Loot_Collect
 volumeMultiplier=0.3462
 concurrencyOverrides {maxCount=1, resolutionRule=PreventNew}
}
